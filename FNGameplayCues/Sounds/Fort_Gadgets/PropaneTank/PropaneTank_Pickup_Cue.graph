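# Sound cue for picking up a propane tank gadget, routed through the player foley sound class.
# A short fixed delay feeds a modulator, which varies an equally weighted random pick from three pickup wave variants.
# It overrides attenuation for falloff and stereo spread, and limits concurrency to a couple of instances, stopping the oldest.
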SoundCue /FNGameplayCues/Sounds/Fort_Gadgets/PropaneTank/PropaneTank_Pickup_Cue
root=Delay_1
nodes [
 {id=Delay_1, children [Modulator_0], properties {delayMax=0.08, delayMin=0.08}}
 {id=Modulator_0, children [Random_4]}
 {id=Random_4, children [WavePlayer_0,WavePlayer_1,WavePlayer_2], properties {weights [1,1,1]}}
 {id=WavePlayer_0, properties {soundWaveAssetPtr=&/PropaneTank_Pickup_01}}
 {id=WavePlayer_1, properties {soundWaveAssetPtr=&/PropaneTank_Pickup_02}}
 {id=WavePlayer_2, properties {soundWaveAssetPtr=&/PropaneTank_Pickup_03}}
]
properties {
 attenuationSettings=/Game/Sounds/Attenuation/Fort_Prop_Close_Att
 bOverrideAttenuation=true
 bOverrideConcurrency=true
 maxDistance=1400
 soundClassObject=&Fort_Player_Foley
 volumeMultiplier=0.9
 attenuationOverrides {falloffDistance=1000, stereoSpread=500}
 concurrencyOverrides {maxCount=2, resolutionRule=StopOldest}
}
