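# Sound cue for throwing a propane tank gadget, played as a quiet positional effect in the ranged weapon sound class.
# A random node picks between two equally weighted throw recordings and feeds a modulator that randomly raises the pitch slightly.
# It borrows the held-object impact attenuation and overrides concurrency so new plays are blocked, not stealing active ones, once the limit is reached.
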SoundCue /FNGameplayCues/Sounds/Fort_Gadgets/PropaneTank/PropaneTank_Throw_Cue
root=Modulator_0
nodes [
 {id=Modulator_0, children [Random_1], properties {pitchMax=1.3, pitchMin=1.2}}
 {id=Random_1, children [WavePlayer_3,WavePlayer_5], properties {weights [1,1]}}
 {id=WavePlayer_3, properties {soundWaveAssetPtr=&/PropaneTank_Throw_01}}
 {id=WavePlayer_5, properties {soundWaveAssetPtr=&/PropaneTank_Throw_03}}
]
properties {
 attenuationSettings=/FNGameplayCues/Sounds/Attenuation/Fort_HeldObject_Impact_Att
 bOverrideConcurrency=true
 maxDistance=1912
 soundClassObject=&Fort_WeaponRanged
 volumeMultiplier=0.1
 concurrencyOverrides {resolutionRule=PreventNew}
}
